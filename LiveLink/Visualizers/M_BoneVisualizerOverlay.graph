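# Unlit translucent overlay material for the Live Link bone visualizer, set up for instanced static meshes.
# Depth testing is disabled, so it draws over occluding geometry.
# A Color vector parameter, green by default, feeds EyeAdaptationInverse, which counteracts auto exposure to keep the color's brightness stable.
# It uses a coarse 4x4 shading rate, and two constant color nodes have no visible connections.
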Material /LiveLink/Visualizers/M_BoneVisualizerOverlay
graph {
 properties {
  bDisableDepthTest=true
  bUsedWithInstancedStaticMeshes=true
  blendMode=BLEND_Translucent
  refractionCoverageMode=RCM_CoverageIgnored
  shadingModel=MSM_Unlit
  shadingRate=MSR_4x4
 }
 nodes [
  Constant3Vector() {constant=<rgba 0.502 0.502 0.502 1>}
  Constant3Vector() {constant=<rgba 1 1 1 0>}
  EyeAdaptationInverse(LightValueInput=VectorParameter() {parameterName=Color, defaultValue=<rgba 0.0163 0.2083 0 1>}.RGB)
 ]
}
