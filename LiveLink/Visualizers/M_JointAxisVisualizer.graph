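# Unlit translucent material for Live Link joint axis visualizers, with depth testing disabled so it shows through geometry and usage enabled for instanced static meshes and Nanite.
# A Color parameter passes through the bUseEmissive static switch and is scaled by EmissiveMult.
# EyeAdaptationInverse takes that value, with the same switch setting its alpha, to offset auto exposure.
# Opacity, Metallic, Specular and Roughness parameters are also defined.
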Material /LiveLink/Visualizers/M_JointAxisVisualizer
graph {
 properties {
  bDisableDepthTest=true
  bUsedWithInstancedStaticMeshes=true
  bUsedWithNanite=true
  blendMode=BLEND_Translucent
  ditherOpacityMask=true
  pixelDepthOffsetMode=PDOM_Legacy
  refractionCoverageMode=RCM_CoverageIgnored
  shadingModel=MSM_Unlit
  translucencyLightingMode=TLM_VolumetricDirectional
 }
 nodes [
  Constant3Vector() {constant=<rgba 1 1 1 0>}
  n1 = VectorParameter() {parameterName=Color, defaultValue=<rgba 1 1 1 1>}
  n2 = Constant()
  EyeAdaptationInverse(
   LightValueInput=Multiply(
    A=StaticSwitchParameter(True=@n1.RGB, False=@n2) {defaultValue=true, parameterName=bUseEmissive}
    B=ScalarParameter() {defaultValue=1, parameterName=EmissiveMult}
   )
   AlphaInput=StaticSwitchParameter(True=Constant() {r=1}, False=Constant()) {defaultValue=true, parameterName=bUseEmissive}
  )
  ScalarParameter() {parameterName=Metallic}
  ScalarParameter() {parameterName=Specular}
  ScalarParameter() {parameterName=Roughness}
  ScalarParameter() {defaultValue=1, parameterName=Opacity}
 ]
}
